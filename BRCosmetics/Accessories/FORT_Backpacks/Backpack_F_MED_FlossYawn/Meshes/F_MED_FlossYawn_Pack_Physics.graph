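# Physics asset for the FlossYawn backpack cosmetic, defining secondary motion for its arms, ears, legs, ribbons and fringe.
# A large offset MSTR_body capsule roots the rig, with dyn_body locked to it and tiny base_* anchor capsules locked to dyn_body.
# Capsule dyn_* chains hang from those anchors with limited swing and tight or locked twist.
# Each side's longer ribbon chain ends in four fringe bodies branching from its last link.
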PhysicsAsset /BRCosmetics/Accessories/FORT_Backpacks/Backpack_F_MED_FlossYawn/Meshes/F_MED_FlossYawn_Pack_Physics
bodies {
 MSTR_body [{shapeType=Capsule, length=22.4854, radius=60, shapeName=MSTR_body_capsule, center=<0 66.4188 -21.586>, rotation=<rot 5.3732 -90 0>}]
 base_arm_l [{shapeType=Capsule, length=0.505, radius=0.505, shapeName=base_arm_l_capsule, rotation=<rot 0 -90 0>}]
 base_arm_r [{shapeType=Capsule, length=0.505, radius=0.505, shapeName=base_arm_r_capsule, rotation=<rot 0 -90 0>}]
 base_ear_l_01 [{shapeType=Capsule, length=0.505, radius=0.505, shapeName=base_ear_l_01_capsule, rotation=<rot 0 -90 0>}]
 base_ear_r_01 [{shapeType=Capsule, length=0.505, radius=0.505, shapeName=base_ear_r_01_capsule, rotation=<rot 0 -90 0>}]
 base_leg_l [{shapeType=Capsule, length=0.505, radius=0.505, shapeName=base_leg_l_capsule, rotation=<rot 0 -90 0>}]
 base_leg_r [{shapeType=Capsule, length=0.505, radius=0.505, shapeName=base_leg_r_capsule, rotation=<rot 0 -90 0>}]
 base_libbond_l_00 [{shapeType=Capsule, length=0.505, radius=0.505, shapeName=base_libbond_l_00_capsule, rotation=<rot 0 -90 0>}]
 base_ribbond_l_01 [{shapeType=Capsule, length=0.505, radius=0.505, shapeName=base_ribbond_l_01_capsule, rotation=<rot 0 -90 0>}]
 base_ribbond_r_00 [{shapeType=Capsule, length=0.505, radius=0.505, shapeName=base_ribbond_r_00_capsule, rotation=<rot 0 -90 0>}]
 base_ribbond_r_01 [{shapeType=Capsule, length=0.505, radius=0.505, shapeName=base_ribbond_r_01_capsule, rotation=<rot 0 -90 0>}]
 dyn_arm_l [{
   shapeType=Capsule
   length=6.5
   radius=1.5
   shapeName=dyn_arm_l_capsule
   center=<1.7951 -0.7126 0.7768>
   rotation=<rot 3.4036 69.7197 -69.4085>
  }]
 dyn_arm_r [{shapeType=Capsule, length=6.5, radius=1.5, shapeName=dyn_arm_r_capsule, center=<2.3052 0.2769 0.82>, rotation=<rot 7.6451 99.5675 -66.4667>}]
 dyn_body [{shapeType=Capsule, length=5.5, radius=2.5, shapeName=dyn_body_capsule, center=<0 -0.5092 -3.7686>, rotation=<rot -90 -0.0001 0>}]
 dyn_ear_l_01 [{shapeType=Capsule, length=4, radius=2.5, shapeName=dyn_ear_l_01_capsule, center=<3 0.2866 0>, rotation=<rot 79.9999 179.9998 179.8481>}]
 dyn_ear_l_02 [{shapeType=Capsule, length=4.5, radius=2, shapeName=dyn_ear_l_02_capsule, center=<-0.497 0 2.8184>, rotation=<rot -10.0003 180 179.9201>}]
 dyn_ear_r_01 [{shapeType=Capsule, length=4, radius=2.5, shapeName=dyn_ear_r_01_capsule, center=<3 -0.3217 0>, rotation=<rot 80 -179.9999 179.8483>}]
 dyn_ear_r_02 [{
   shapeType=Capsule
   length=2
   radius=2
   shapeName=dyn_ear_r_02_capsule
   center=<2.5939 -0.3056 1.1156>
   rotation=<rot 69.9992 -179.9993 179.8488>
  }]
 dyn_fringe_l_01 [{shapeType=Capsule, length=4.5, radius=1, shapeName=dyn_fringe_l_01_capsule, center=<2.8292 -0.182 -1.2666>, rotation=<rot 70.0002 0 4.4338>}]
 dyn_fringe_l_02 [{shapeType=Capsule, length=4.5, radius=1, shapeName=dyn_fringe_l_02_capsule, center=<3.1051 -0.506 -1.19>, rotation=<rot 70.0002 0 10>}]
 dyn_fringe_l_03 [{shapeType=Capsule, length=4, radius=1, shapeName=dyn_fringe_l_03_capsule, center=<3.0302 -0.4893 -1.3356>, rotation=<rot 70.0001 0 10>}]
 dyn_fringe_l_04 [{
   shapeType=Capsule
   length=4.5
   radius=1
   shapeName=dyn_fringe_l_04_capsule
   center=<3.1148 -0.6966 -1.1737>
   rotation=<rot 70.0002 0 10.0001>
  }]
 dyn_fringe_r_01 [{shapeType=Capsule, length=4.5, radius=1, shapeName=dyn_fringe_r_01_capsule, center=<2.5678 0.1756 -0.838>, rotation=<rot 70.0006 0 -3.0987>}]
 dyn_fringe_r_02 [{
   shapeType=Capsule
   length=4
   radius=1
   shapeName=dyn_fringe_r_02_capsule
   center=<2.7715 0.3061 -0.8425>
   rotation=<rot 70.083 -0.016 -8.7854>
  }]
 dyn_fringe_r_03 [{
   shapeType=Capsule
   length=4
   radius=1
   shapeName=dyn_fringe_r_03_capsule
   center=<2.7524 0.5321 -0.8121>
   rotation=<rot 70.0005 -0.0001 -10.0002>
  }]
 dyn_fringe_r_04 [{shapeType=Capsule, length=4, radius=1, shapeName=dyn_fringe_r_04_capsule, center=<2.8058 0.308 -1.1026>, rotation=<rot 70.0002 0 -3.3295>}]
 dyn_leg_l_01 [{shapeType=Capsule, length=2.5, radius=2.5, shapeName=dyn_leg_l_01_capsule, center=<3.065 0 0>, rotation=<rot -90 0.0001 0>}]
 dyn_leg_l_02 [{shapeType=Capsule, length=4, radius=1.5, shapeName=dyn_leg_l_02_capsule, center=<2.1371 -0.1671 0>, rotation=<rot 90 0.4678 0>}]
 dyn_leg_r_01 [{shapeType=Capsule, length=2.5, radius=2.5, shapeName=dyn_leg_r_01_capsule, center=<3.065 0 0>, rotation=<rot -90 0.0001 0>}]
 dyn_leg_r_02 [{shapeType=Capsule, length=4, radius=1.5, shapeName=dyn_leg_r_02_capsule, center=<2.1371 -0.1671 0>, rotation=<rot 90 0.4679 0>}]
 dyn_libbond_l_00 [{shapeType=Capsule, length=5, radius=3, shapeName=dyn_libbond_l_00_capsule, center=<3.7744 0 0>, rotation=<rot -90 0.0002 0>}]
 dyn_ribbond_l_01 [{
   shapeType=Capsule
   length=5
   radius=2.5
   shapeName=dyn_ribbond_l_01_capsule
   center=<3.4956 0 -0.3318>
   rotation=<rot -81.7569 -179.9998 179.9998>
  }]
 dyn_ribbond_l_02 [{shapeType=Capsule, length=2, radius=4, shapeName=dyn_ribbond_l_02_capsule, center=<3.9182 -0.3125 -1.9833>, rotation=<rot -90 0 0>}]
 dyn_ribbond_l_03 [{
   shapeType=Capsule
   length=2
   radius=4
   shapeName=dyn_ribbond_l_03_capsule
   center=<4.0462 -0.5972 -2.4272>
   rotation=<rot 82.0173 0.0001 -179.9999>
  }]
 dyn_ribbond_r_00 [{shapeType=Capsule, length=5, radius=3, shapeName=dyn_ribbond_r_00_capsule, center=<3.7744 0 0>, rotation=<rot -90 -0.0002 0>}]
 dyn_ribbond_r_01 [{
   shapeType=Capsule
   length=5
   radius=2.5
   shapeName=dyn_ribbond_r_01_capsule
   center=<3.4956 0 -0.3318>
   rotation=<rot -81.7569 179.9997 -179.9998>
  }]
 dyn_ribbond_r_02 [{shapeType=Capsule, length=2, radius=4, shapeName=dyn_ribbond_r_02_capsule, center=<3.9182 0.5312 -1.9833>, rotation=<rot -90 0 0>}]
 dyn_ribbond_r_03 [{
   shapeType=Capsule
   length=1.5
   radius=4
   shapeName=dyn_ribbond_r_03_capsule
   center=<3.9104 0.5768 -2.0343>
   rotation=<rot 80.2214 -0.0001 179.9999>
  }]
}
constraints [
 {
  bone1Name=base_arm_l
  bone2Name=dyn_body
  swing1LimitDegrees=45
  swing1Motion=Locked
  swing2LimitDegrees=45
  swing2Motion=Locked
  twistLimitDegrees=45
  twistMotion=Locked
 }
 {bone1Name=dyn_arm_l, bone2Name=base_arm_l, swing1LimitDegrees=30, swing2LimitDegrees=30, twistLimitDegrees=2}
 {
  bone1Name=base_arm_r
  bone2Name=dyn_body
  swing1LimitDegrees=45
  swing1Motion=Locked
  swing2LimitDegrees=45
  swing2Motion=Locked
  twistLimitDegrees=45
  twistMotion=Locked
 }
 {bone1Name=dyn_arm_r, bone2Name=base_arm_r, swing1LimitDegrees=30, swing2LimitDegrees=30, twistLimitDegrees=2}
 {
  bone1Name=base_ear_l_01
  bone2Name=dyn_body
  swing1LimitDegrees=45
  swing1Motion=Locked
  swing2LimitDegrees=45
  swing2Motion=Locked
  twistLimitDegrees=45
  twistMotion=Locked
 }
 {bone1Name=dyn_ear_l_01, bone2Name=base_ear_l_01, swing1LimitDegrees=50, swing2LimitDegrees=20, twistLimitDegrees=2}
 {bone1Name=dyn_ear_l_02, bone2Name=dyn_ear_l_01, swing1LimitDegrees=30, swing2LimitDegrees=10, twistLimitDegrees=2}
 {
  bone1Name=base_ear_r_01
  bone2Name=dyn_body
  swing1LimitDegrees=45
  swing1Motion=Locked
  swing2LimitDegrees=45
  swing2Motion=Locked
  twistLimitDegrees=45
  twistMotion=Locked
 }
 {bone1Name=dyn_ear_r_01, bone2Name=base_ear_r_01, swing1LimitDegrees=50, swing2LimitDegrees=20, twistLimitDegrees=2}
 {bone1Name=dyn_ear_r_02, bone2Name=dyn_ear_r_01, swing1LimitDegrees=29, swing2LimitDegrees=10, twistLimitDegrees=2}
 {
  bone1Name=base_leg_l
  bone2Name=dyn_body
  swing1LimitDegrees=45
  swing1Motion=Locked
  swing2LimitDegrees=45
  swing2Motion=Locked
  twistLimitDegrees=45
  twistMotion=Locked
 }
 {bone1Name=dyn_leg_l_01, bone2Name=base_leg_l, swing1LimitDegrees=20, swing2LimitDegrees=10, twistLimitDegrees=2}
 {bone1Name=dyn_leg_l_02, bone2Name=dyn_leg_l_01, swing1LimitDegrees=20, swing2LimitDegrees=30, twistLimitDegrees=2}
 {
  bone1Name=base_leg_r
  bone2Name=dyn_body
  swing1LimitDegrees=45
  swing1Motion=Locked
  swing2LimitDegrees=45
  swing2Motion=Locked
  twistLimitDegrees=45
  twistMotion=Locked
 }
 {bone1Name=dyn_leg_r_02, bone2Name=dyn_leg_r_01, swing1LimitDegrees=20, swing2LimitDegrees=30, twistLimitDegrees=2}
 {
  bone1Name=base_libbond_l_00
  bone2Name=dyn_body
  swing1LimitDegrees=45
  swing1Motion=Locked
  swing2LimitDegrees=45
  swing2Motion=Locked
  twistLimitDegrees=45
  twistMotion=Locked
 }
 {bone1Name=dyn_leg_r_01, bone2Name=base_leg_r, swing1LimitDegrees=20, swing2LimitDegrees=10, twistLimitDegrees=2}
 {bone1Name=dyn_libbond_l_00, bone2Name=base_libbond_l_00, swing1LimitDegrees=20, swing2LimitDegrees=40, twistLimitDegrees=2}
 {
  bone1Name=base_ribbond_l_01
  bone2Name=dyn_body
  swing1LimitDegrees=45
  swing1Motion=Locked
  swing2LimitDegrees=45
  swing2Motion=Locked
  twistLimitDegrees=45
  twistMotion=Locked
 }
 {
  bone1Name=dyn_ribbond_l_01
  bone2Name=base_ribbond_l_01
  swing1LimitDegrees=30
  swing2LimitDegrees=50
  twistLimitDegrees=2
  twistMotion=Locked
 }
 {
  bone1Name=dyn_ribbond_l_02
  bone2Name=dyn_ribbond_l_01
  swing1LimitDegrees=40
  swing2LimitDegrees=50
  twistLimitDegrees=2
  twistMotion=Locked
 }
 {
  bone1Name=dyn_ribbond_l_03
  bone2Name=dyn_ribbond_l_02
  swing1LimitDegrees=50
  swing2LimitDegrees=20
  twistLimitDegrees=2
  twistMotion=Locked
 }
 {bone1Name=dyn_fringe_l_01, bone2Name=dyn_ribbond_l_03, swing1LimitDegrees=20, swing2LimitDegrees=20, twistLimitDegrees=2, twistMotion=Locked}
 {bone1Name=dyn_fringe_l_02, bone2Name=dyn_ribbond_l_03, swing1LimitDegrees=20, swing2LimitDegrees=20, twistLimitDegrees=2, twistMotion=Locked}
 {bone1Name=dyn_fringe_l_03, bone2Name=dyn_ribbond_l_03, swing1LimitDegrees=20, swing2LimitDegrees=20, twistLimitDegrees=2, twistMotion=Locked}
 {bone1Name=dyn_fringe_l_04, bone2Name=dyn_ribbond_l_03, swing1LimitDegrees=20, swing2LimitDegrees=20, twistLimitDegrees=2, twistMotion=Locked}
 {
  bone1Name=base_ribbond_r_00
  bone2Name=dyn_body
  swing1LimitDegrees=45
  swing1Motion=Locked
  swing2LimitDegrees=45
  swing2Motion=Locked
  twistLimitDegrees=45
  twistMotion=Locked
 }
 {bone1Name=dyn_ribbond_r_00, bone2Name=base_ribbond_r_00, swing1LimitDegrees=20, swing2LimitDegrees=40, twistLimitDegrees=2}
 {
  bone1Name=base_ribbond_r_01
  bone2Name=dyn_body
  swing1LimitDegrees=45
  swing1Motion=Locked
  swing2LimitDegrees=45
  swing2Motion=Locked
  twistLimitDegrees=45
  twistMotion=Locked
 }
 {
  bone1Name=dyn_ribbond_r_01
  bone2Name=base_ribbond_r_01
  swing1LimitDegrees=30
  swing2LimitDegrees=50
  twistLimitDegrees=2
  twistMotion=Locked
 }
 {
  bone1Name=dyn_ribbond_r_02
  bone2Name=dyn_ribbond_r_01
  swing1LimitDegrees=40
  swing2LimitDegrees=50
  twistLimitDegrees=2
  twistMotion=Locked
 }
 {
  bone1Name=dyn_ribbond_r_03
  bone2Name=dyn_ribbond_r_02
  swing1LimitDegrees=50
  swing2LimitDegrees=20
  twistLimitDegrees=2
  twistMotion=Locked
 }
 {bone1Name=dyn_fringe_r_01, bone2Name=dyn_ribbond_r_03, swing1LimitDegrees=20, swing2LimitDegrees=20, twistLimitDegrees=2, twistMotion=Locked}
 {bone1Name=dyn_fringe_r_02, bone2Name=dyn_ribbond_r_03, swing1LimitDegrees=20, swing2LimitDegrees=20, twistLimitDegrees=2, twistMotion=Locked}
 {bone1Name=dyn_fringe_r_03, bone2Name=dyn_ribbond_r_03, swing1LimitDegrees=20, swing2LimitDegrees=20, twistLimitDegrees=2, twistMotion=Locked}
 {bone1Name=dyn_fringe_r_04, bone2Name=dyn_ribbond_r_03, swing1LimitDegrees=20, swing2LimitDegrees=20, twistLimitDegrees=2, twistMotion=Locked}
 {
  bone1Name=dyn_body
  bone2Name=MSTR_body
  swing1LimitDegrees=45
  swing1Motion=Locked
  swing2LimitDegrees=45
  swing2Motion=Locked
  twistLimitDegrees=45
  twistMotion=Locked
 }
]
