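# Animation Blueprint for the FlossYawn backpack cosmetic, adding procedural offsets and rigid-body secondary motion.
# Each update maps spine_05 yaw relative to spine_01 into additive roll and sideways offset on the pack bone.
# The adjusted pose feeds a RigidBody node simulating relative to MSTR_body with fixed gravity.
# A nativized gravity-override function and front-end branch set variables, but only AlphaFE reaches that node, at 1 either way.
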AnimBlueprint /BRCosmetics/Accessories/FORT_Backpacks/Backpack_F_MED_FlossYawn/Meshes/F_MED_FlossYawn_Pack_AnimBP
target_skeleton=&/F_MED_FlossYawn_Pack_Skeleton
graphs [
 AnimGraph {nodes [
   n1 = "Transform (Modify) Bone - Bone: pack"(
    ComponentPose=Local To Component(LocalPose=Use cached pose 'Input_pose'())
    Translation=Get backpack_translation()
    Rotation=Get backpack_rotate()
   ) {
    rotationMode=BMM_Additive
    rotationSpace=BCS_BoneSpace
    translationMode=BMM_Additive
    translationSpace=BCS_BoneSpace
    alphaBoolBlend {blendOption=Linear}
    boneToModify {boneName=pack}
   }
   n2 = RigidBody!(
    ComponentPose=@n1
    Alpha=Get AlphaFE() -> (AlphaFE:double)
    OverrideWorldGravity:Vector="0,0,-980"
    SimSpaceSettings:Sim Space Settings Structure=(WorldAlpha=0.35,VelocityScaleZ=0.2,DampingAlpha=1,MaxLinearVelocity=400,MaxAngularVelocity=5,MaxLinearAcceleration=1500,MaxAngularAcceleration=30,ExternalLinearDragV=(X=0.1,Y=0.1,Z=0.1),ExternalLinearVelocity=(X=0,Y=0,Z=0),ExternalAngularVelocity=(X=0,Y=0,Z=0))
   ) -> (Pose) {
    bClampLinearTranslationLimitToRefPose=true
    bOverrideWorldGravity=true
    bUseExternalClothCollision=true
    overridePhysicsAsset=&/F_MED_FlossYawn_Pack_Physics
    simulationSpace=BaseBoneSpace
    alphaBoolBlend {blendOption=Linear}
    baseBoneRef {boneName=MSTR_body}
    componentAppliedLinearAccClamp=<800 800 800>
    componentLinearAccScale=<0.3 0.3 0.2>
    componentLinearVelScale=<0.3 0.3 0.2>
    simSpaceSettings {
     maxAngularAcceleration=30
     maxAngularVelocity=5
     maxLinearAcceleration=1500
     maxLinearVelocity=400
     velocityScaleZ=0.2
     worldAlpha=0.35
     externalLinearDragV=<0.1 0.1 0.1>
    }
   }
   Output Pose(Result=Component To Local(ComponentPose=@n2)) {layerGroup=None}
   Save cached pose 'Input_pose'(Pose=Input Pose()) {cacheName=Input_pose}
  ]}
 EventGraph {nodes [
   Event BlueprintUpdateAnimation()
   > n1 = Sequence!(execute) -> (then_0, then_1, then_2:Exec)
   > n2 = "Gravity Override State Function (Parameter Asset) (Nativized)"(
    CharacterPartAnimInstance=Self-Reference()
    GameTimeInSeconds=GetGameTimeInSeconds()
    DebugJointName=spine_05
    Parameters=&/F_MED_FlossYawn_Pack__Parameters
   )
   > Set GravityOverride(GravityOverride=@n2)
   > Set RBAN Settings Out(RBAN Settings Out=@n2.OutRigidBodyAnimNodeParameters)
   n3 = GetOwnerSkeletalMeshComponent()
   n4 = BreakRotator(InRot=BreakTransform(InTransform=GetDeltaTransformFromRefPose(self=@n3, BoneName=spine_05, BaseName=spine_01)).Rotation)
   Set backpack_rotate(execute=@n1.then_1, backpack_rotate=MakeRotator(Roll=MapRangeClamped(Value=@n4.Yaw, InRangeA=0, InRangeB=-30, OutRangeB=45)))
   > Set backpack_translation(backpack_translation=MakeVector(Y=MapRangeClamped(Value=@n4.Yaw, InRangeA=0, InRangeB=-30, OutRangeA=-1.2, OutRangeB=-3)))
   > n5 = Branch(Condition=Get bIsInFrontEnd())
   > Set SimlationSpace(SimlationSpace:AnimPhysSimSpaceType Enum=World) -> (Output_Get:AnimPhysSimSpaceType Enum)
   > Set AlphaFE(AlphaFE:double=1) -> (Output_Get:double)
   Set SimlationSpace(execute=@n5.else)
   > Set AlphaFE(AlphaFE=1)
  ]}
]
variables [
 AlphaFE {type=number}
 backpack_rotate {type=Rotator}
 backpack_translation {type=Vector}
 GravityOverride {type=Vector}
 RBAN Settings Out {type=FortRigidBodyAnimNodeParameters}
 SimlationSpace {type=AnimPhysSimSpaceType}
]
